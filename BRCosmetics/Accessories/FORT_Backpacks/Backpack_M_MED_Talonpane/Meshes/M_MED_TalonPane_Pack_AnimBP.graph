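# Animation Blueprint for the Talonpane backpack that adjusts the pack's bones, except while skydiving, when it passes the input pose through.
# The AnimGraph offsets the pack bone, has additive armor-bone modifiers with zero rotation, and aims both pistons at target bones via Look At.
# The update event derives mirrored, clamped armor pitch from averaged upper-arm rotations, but no AnimGraph node visibly reads it.
# A second, per-arm variant is left unconnected.
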AnimBlueprint /BRCosmetics/Accessories/FORT_Backpacks/Backpack_M_MED_Talonpane/Meshes/M_MED_TalonPane_Pack_AnimBP
target_skeleton=&/M_MED_TalonPane_Pack_Skeleton
graphs [
 AnimGraph {nodes [
   n1 = "Transform (Modify) Bone - Bone: pack"(ComponentPose=Local To Component(LocalPose=Input Pose()), Translation="0,2,0", Rotation="0,0,0") {
    translationMode=BMM_Additive
    translationSpace=BCS_BoneSpace
    alphaBoolBlend {blendOption=Linear}
    boneToModify {boneName=pack}
    translation {y=2}
   }
   n2 = Look At - Bone: dyn_piston_01_r(
    ComponentPose="Transform (Modify) Bone - Bone: dyn_armor_01_l"(
     ComponentPose="Transform (Modify) Bone - Bone: dyn_armor_01_r"(ComponentPose=@n1, Alpha=1, Translation="0,0,0", Rotation="0,0,0", Scale="1,1,1") {rotationMode=BMM_Additive, alphaBoolBlend {blendOption=Linear}, boneToModify {boneName=dyn_armor_01_r}}
     Alpha=1
     Translation="0,0,0"
     Rotation="0,0,0"
     Scale="1,1,1"
    ) {rotationMode=BMM_Additive, alphaBoolBlend {blendOption=Linear}, boneToModify {boneName=dyn_armor_01_l}}
    Alpha=1
   ) {
    alphaBoolBlend {blendOption=Linear}
    boneToModify {boneName=dyn_piston_01_r}
    lookAtLocation {x=0}
    lookAtTarget {boneReference {boneName=dyn_target_piston_r}}
    lookAt_Axis {axis=<1 0>}
   }
   n3 = Look At - Bone: dyn_piston_01_l(ComponentPose=@n2, Alpha=1) {
    alphaBoolBlend {blendOption=Linear}
    boneToModify {boneName=dyn_piston_01_l}
    lookAtLocation {x=0}
    lookAtTarget {boneReference {boneName=dyn_target_piston_l}}
    lookAt_Axis {axis=<-1 0>}
   }
   Output Pose(Result=Blend Poses by bool(BlendPose_0=Input Pose() {name=InPose_0}, BlendPose_1=Component To Local(ComponentPose=@n3), bActiveValue=Get bIsSkyDiving()) {blendType=Linear}) {layerGroup=None}
  ]}
 EventGraph {nodes [
   Event BlueprintUpdateAnimation()
   n1 = GetOwnerSkeletalMeshComponent()
   > Set UpperarmRotationL(UpperarmRotationL:Rotator=GetBoneTransform!(self:Skinned Mesh Component Object Reference=@n1, InBoneName:Name=upperarm_l, TransformSpace:ERelativeTransformSpace Enum=RTS_World) -> (ReturnValue_Location:Vector, ReturnValue_Rotation:Rotator, ReturnValue_Scale:Vector).ReturnValue_Rotation) -> (Output_Get:Rotator)
   > Set UpperarmRotationR(UpperarmRotationR:Rotator="Scale Rotator (Float)"(A:Rotator=GetBoneTransform(self=@n1, InBoneName=upperarm_r, TransformSpace=RTS_World).ReturnValue_Rotation, B:float=-1) -> (ReturnValue:Rotator=~) {is=title K2Node_PromotableOperator}) -> (Output_Get:Rotator)
   n2 = float + float(
    A=BreakRotator(InRot=Get UpperarmRotationL() -> (UpperarmRotationL:Rotator)).Pitch
    B=BreakRotator(InRot=Get UpperarmRotationR() -> (UpperarmRotationR:Rotator)).Pitch
   )
   n3 = MakeRotator(Pitch="Clamp (Float)"(Value=float / float(A=@n2, B=2), Min=-20, Max=20))
   > Set ArmorRotationL(ArmorRotationL:Rotator=@n3) -> (Output_Get:Rotator)
   > Set ArmorRotationR(ArmorRotationR:Rotator="Scale Rotator (Float)"(A=@n3, B=-1) -> (ReturnValue=~)) -> (Output_Get:Rotator)
   n4 = GetOwnerSkeletalMeshComponent()
   Set UpperarmRotationL(UpperarmRotationL=GetBoneTransform(self=@n4, InBoneName=upperarm_l, TransformSpace=RTS_World).ReturnValue_Rotation)
   > Set UpperarmRotationR(UpperarmRotationR=GetBoneTransform(self=@n4, InBoneName=upperarm_r, TransformSpace=RTS_World).ReturnValue_Rotation)
   n5 = BreakRotator(InRot=Get UpperarmRotationL())
   > Set ArmorRotationL(ArmorRotationL=MakeRotator(Pitch="Clamp (Float)"(Value=@n5.Pitch, Min=-20, Max=20)))
   n6 = BreakRotator(InRot=Get UpperarmRotationR())
   > Set ArmorRotationR(ArmorRotationR=MakeRotator(Pitch="Clamp (Float)"(Value=@n6.Pitch, Min=-20, Max=20)))
  ]}
]
variables [
 ArmorRotationL {type=Rotator}
 ArmorRotationR {type=Rotator}
 UpperarmRotationL {type=Rotator}
 UpperarmRotationR {type=Rotator}
]
